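; Editor viewmodel for a selected Sound asset that adds or removes StreamSafe mute tags stored as Audio Asset User Data.
; Field-notify getters expose whether a sound is set and its user data, rebroadcast after each edit.
; Add and Remove check out the asset, call BPAudioMetaDataEditorFunctionLibrary inside an undo transaction, then save on success.
; An editor notification reports checkout, tagging and save results.
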(fn AddStreamSafeTags
  (bind _returnvalue (EditorScripting|Notifications|AddEditorNotification (LocalVariables|GetEditor%20Notification%20Settings)))
  (bind _output_get (LocalVariables|SetEditor%20Notification _returnvalue))
  (EditorScripting|Notifications|SetCompletionState _output_get "Pending")
  (bind _returnvalue_2 (EditorScripting|Asset|CheckoutLoadedAsset (Variables|Default|GetSound)))
  (if _returnvalue_2
    (Transactions|CreateTransaction "StreamSafe Tags" "NSLOCTEXT(\"[4D6D15C02C2C7585547ED693566F5CB3]\", \"A20309F346F1E4A7FB85F9A6B59F1540\", \"Add StreamSafe Tags\")"
      (:Pending
        (bind _sound (Variables|Default|GetSound)
          (:then
            (Transactions|TransactObject _sound)
            (bind _returnvalue_3 (EditorScripting|LevelUtility|GetEditorWorld (EditorSubsystems|GetUnrealEditorSubsystem)))
            (bind _success (Class|BPAudioMetaDataEditorFunctionLibrary|AddMuteforStreamSafeTag _sound _returnvalue_3))
            (bind _output_get_2 (LocalVariables|SetSuccessfully%20Added%20Tags _success))
            (if _output_get_2
              (EditorScripting|Asset|SetDirtyFlag (EditorSubsystems|GetEditorAssetSubsystem) _sound true)
              (bind _returnvalue_4 (EditorScripting|Asset|SaveLoadedAsset (Variables|Default|GetSound) :!class /Script/EditorScriptingUtilities.EditorAssetLibrary))
              (LocalVariables|SetSuccessfully%20Saved _returnvalue_4))
            (Utilities|Debugging|RaiseScriptError "Failed to add StreamSafe tags."))
          (:else
            _))
        (Utilities|Debugging|RaiseScriptError "No sound selected. Select a sound and try again."))
      (:Complete
        _))
    (FieldNotification|BroadcastFieldValueChanged self "(FieldName=\"Get Audio Asset User Data\")")
    (FieldNotification|BroadcastFieldValueChanged self "(FieldName=\"Has Audio Asset User Data\")")
    (LocalVariables|GetSuccessfully%20Added%20Tags
      (:then
        (LocalVariables|GetSuccessfully%20Saved
          (:then
            (EditorScripting|Notifications|SetSubText (LocalVariables|GetEditor%20Notification) "NSLOCTEXT(\"[4D6D15C02C2C7585547ED693566F5CB3]\", \"6F54DD9F4FD8331E332DDDAC346800D1\", \"Asset was successfully checked out, modified, and saved.\")"))
          (:else
            (EditorScripting|Notifications|SetSubText (LocalVariables|GetEditor%20Notification) (Utilities|Text|FormatText "NSLOCTEXT(\"[4D6D15C02C2C7585547ED693566F5CB3]\", \"FDD266DF464E8A939EDF7A93A9B9EB30\", \"Tags were added but failed to save asset. Please manually save the asset before submitting.\")"))
            (EditorScripting|Notifications|SetExpireDuration (LocalVariables|GetEditor%20Notification) 3e+01)))
        (EditorScripting|Notifications|SetText (LocalVariables|GetEditor%20Notification) "NSLOCTEXT(\"[4D6D15C02C2C7585547ED693566F5CB3]\", \"BAE434C04FDC49493D5F33A167276879\", \"Successfully added StreamSafe Tags!\")")
        (EditorScripting|Notifications|SetCompletionState (LocalVariables|GetEditor%20Notification) "Success")
        (EditorScripting|Notifications|ExpireandFadeOut (LocalVariables|GetEditor%20Notification))
        (return true))
      (:else
        (EditorScripting|Notifications|SetSubText (LocalVariables|GetEditor%20Notification) (Utilities|Text|FormatText "NSLOCTEXT(\"[4D6D15C02C2C7585547ED693566F5CB3]\", \"4E10428E453F44852C7CA0B9ABEE4397\", \"Failed to add StreamSafe tags. Check logs for more information.\")"))
        (EditorScripting|Notifications|SetCompletionState (LocalVariables|GetEditor%20Notification) "Fail")
        (EditorScripting|Notifications|SetText (LocalVariables|GetEditor%20Notification) (Utilities|Text|FormatText "NSLOCTEXT(\"[4D6D15C02C2C7585547ED693566F5CB3]\", \"6EBBFDB14739475D709A8FAB737328CF\", \"Did not add StreamSafe tags!\")"))
        (EditorScripting|Notifications|ExpireandFadeOut (LocalVariables|GetEditor%20Notification))
        (return false)))
    (else
      (EditorScripting|Notifications|SetSubText (LocalVariables|GetEditor%20Notification) (Utilities|Text|FormatText "NSLOCTEXT(\"[4D6D15C02C2C7585547ED693566F5CB3]\", \"97859E334484087E23D3C7874DF0DDFE\", \"Failed to check out asset! Ensure the asset is available for checkout and try again.\")"))
      (EditorScripting|Notifications|SetText (LocalVariables|GetEditor%20Notification) (Utilities|Text|FormatText "NSLOCTEXT(\"[4D6D15C02C2C7585547ED693566F5CB3]\", \"86A326A043CFF9B1BC96A98B9DEC30FD\", \"Did not add StreamSafe tags!\")"))
      (EditorScripting|Notifications|SetCompletionState (LocalVariables|GetEditor%20Notification) "Fail")
      (EditorScripting|Notifications|ExpireandFadeOut (LocalVariables|GetEditor%20Notification))
      (return false))))

(fn GetAudioAssetUserData
  (bind _sound (Variables|Default|GetSound)
    (:then
      (bind _returnvalue (AssetUserData|GetAssetUserDataOfClass _sound (Variables|Default|GetUserDataClass)))
      (return (Utilities|Casting|CastToAudioAssetUserData _returnvalue)))
    (:else
      (return 0))))

(fn HasAsset
  (Variables|Default|GetSound
    (:then
      (return true))
    (:else
      (return false))))

(fn HasAudioAssetUserData
  (bind _sound (Variables|Default|GetSound)
    (:then
      (bind _returnvalue (AssetUserData|GetAssetUserDataOfClass _sound (Variables|Default|GetUserDataClass)))
      (bind _asaudio_asset_user_data (Utilities|Casting|CastToAudioAssetUserData _returnvalue)
        (:then
          (return (Utilities|IsValid _asaudio_asset_user_data :!class /Script/Engine.KismetSystemLibrary)))
        (:CastFailed
          (return false))))
    (:else
      (return false))))

(fn RemoveStreamSafeTags
  (bind _returnvalue (EditorScripting|Notifications|AddEditorNotification (LocalVariables|GetEditor%20Notification%20Settings)))
  (bind _output_get (LocalVariables|SetEditor%20Notification _returnvalue))
  (EditorScripting|Notifications|SetCompletionState _output_get "Pending")
  (bind _returnvalue_2 (EditorScripting|Asset|CheckoutLoadedAsset (Variables|Default|GetSound)))
  (if _returnvalue_2
    (Transactions|CreateTransaction "StreamSafe Tags" "NSLOCTEXT(\"[4D6D15C02C2C7585547ED693566F5CB3]\", \"C006B02B42210FF7440E3EBC123CE1AE\", \"Remove StreamSafe Tags\")"
      (:Pending
        (bind _sound (Variables|Default|GetSound)
          (:then
            (Transactions|TransactObject _sound)
            (bind _returnvalue_3 (EditorScripting|LevelUtility|GetEditorWorld (EditorSubsystems|GetUnrealEditorSubsystem)))
            (bind _success (Class|BPAudioMetaDataEditorFunctionLibrary|RemoveMuteforStreamSafeTag _sound _returnvalue_3))
            (bind _output_get_2 (LocalVariables|SetSuccessfully%20Removed%20Tags _success))
            (if _output_get_2
              (EditorScripting|Asset|SetDirtyFlag (EditorSubsystems|GetEditorAssetSubsystem) _sound true)
              (bind _returnvalue_4 (EditorScripting|Asset|SaveLoadedAsset (Variables|Default|GetSound) :!class /Script/EditorScriptingUtilities.EditorAssetLibrary))
              (LocalVariables|SetSuccessfully%20Saved _returnvalue_4))
            (Utilities|Debugging|RaiseScriptError "Failed to remove StreamSafe tags."))
          (:else
            _))
        (Utilities|Debugging|RaiseScriptError "No sound selected. Select a sound and try again."))
      (:Complete
        _))
    (FieldNotification|BroadcastFieldValueChanged self "(FieldName=\"Get Audio Asset User Data\")")
    (FieldNotification|BroadcastFieldValueChanged self "(FieldName=\"Has Audio Asset User Data\")")
    (LocalVariables|GetSuccessfully%20Removed%20Tags
      (:then
        (LocalVariables|GetSuccessfully%20Saved
          (:then
            (EditorScripting|Notifications|SetSubText (LocalVariables|GetEditor%20Notification) "NSLOCTEXT(\"[4D6D15C02C2C7585547ED693566F5CB3]\", \"5A2544614CB852011E92D1A6DE59A047\", \"Asset was successfully checked out, modified, and saved.\")"))
          (:else
            (EditorScripting|Notifications|SetSubText (LocalVariables|GetEditor%20Notification) (Utilities|Text|FormatText "NSLOCTEXT(\"[4D6D15C02C2C7585547ED693566F5CB3]\", \"3C4EDE314C6AB979F2ED0595A0C02640\", \"Tags were added but failed to save asset. Please manually save and check out the asset before submitting.\")"))
            (EditorScripting|Notifications|SetExpireDuration (LocalVariables|GetEditor%20Notification) 3e+01)))
        (EditorScripting|Notifications|SetText (LocalVariables|GetEditor%20Notification) "NSLOCTEXT(\"[4D6D15C02C2C7585547ED693566F5CB3]\", \"69EF18B346559DF22A3B97A95339AB8A\", \"Successfully removed StreamSafe Tags!\")")
        (EditorScripting|Notifications|SetCompletionState (LocalVariables|GetEditor%20Notification) "Success")
        (EditorScripting|Notifications|ExpireandFadeOut (LocalVariables|GetEditor%20Notification))
        (return true))
      (:else
        (EditorScripting|Notifications|SetSubText (LocalVariables|GetEditor%20Notification) (Utilities|Text|FormatText "NSLOCTEXT(\"[4D6D15C02C2C7585547ED693566F5CB3]\", \"A52F0B55495C587F689E5692C75105D5\", \"Failed to remove StreamSafe tags. Please open the sound asset and clear the Asset User Data array, then save the asset.\")"))
        (EditorScripting|Notifications|SetCompletionState (LocalVariables|GetEditor%20Notification) "Fail")
        (EditorScripting|Notifications|SetText (LocalVariables|GetEditor%20Notification) (Utilities|Text|FormatText "NSLOCTEXT(\"[4D6D15C02C2C7585547ED693566F5CB3]\", \"EE13E4C74D70C6EA7F24D7AF134A755A\", \"Did not remove StreamSafe tags!\")"))
        (EditorScripting|Notifications|ExpireandFadeOut (LocalVariables|GetEditor%20Notification))
        (return false)))
    (else
      (EditorScripting|Notifications|SetSubText (LocalVariables|GetEditor%20Notification) (Utilities|Text|FormatText "NSLOCTEXT(\"[4D6D15C02C2C7585547ED693566F5CB3]\", \"E37FA73147243E40CCF367A014B701A6\", \"Failed to check out asset! Ensure the asset is available for checkout and try again.\")"))
      (EditorScripting|Notifications|SetCompletionState (LocalVariables|GetEditor%20Notification) "Fail")
      (EditorScripting|Notifications|SetText (LocalVariables|GetEditor%20Notification) (Utilities|Text|FormatText "NSLOCTEXT(\"[4D6D15C02C2C7585547ED693566F5CB3]\", \"95138F9A43CEB8BA3E74029987669A9A\", \"Did not remove StreamSafe tags!\")"))
      (EditorScripting|Notifications|ExpireandFadeOut (LocalVariables|GetEditor%20Notification))
      (return false))))

(fn SetSound
  (params (Sound :type Sound%20Base%20Object%20Reference))
  (Variables|Default|SetSound Sound))
