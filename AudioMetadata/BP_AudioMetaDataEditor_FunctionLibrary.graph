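# Editor utility function library that adds or removes stream-safe muting metadata on Sound Base assets.
# Add ensures AudioAssetUserData, applies MuteForStreamSafe, then registers and attaches a per-asset StreamSafeExceptions tag in the owning plugin's tag ini.
# It warns without source control and reports errors for existing tags or sounds outside plugins.
# Remove calls ResetEditorProperty for AssetUserData on that user data, leaving the ini untouched.
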EditorUtilityBlueprint /AudioMetadata/BP_AudioMetaDataEditor_FunctionLibrary
graphs [
 AddMuteForStreamSafeTag {nodes [
   n1 = AddMuteForStreamSafeTag!() -> (then, AudioAsset:Sound Base Object Reference) {is=title K2Node_FunctionEntry}
   > n2 = Branch(Condition=IsEnabled!() -> (ReturnValue:bool))
   Is Valid(exec=[@n2.then, @n29.Yes], InputObject=@n1)
   > n3 = Set SoundBase(SoundBase:Sound Base Object Reference=@n1) -> (Output_Get:Sound Base Object Reference)
   > n4 = GetAssetUserDataOfClass!(
    execute
    self:Interface Asset User Data Interface=@n3
    InUserDataClass:Asset User Data Class Reference=Get UserDataClass() -> (UserDataClass:Asset User Data Class Reference)
   ) -> (then, ReturnValue:Asset User Data Object Reference)
   > n5 = Cast To AudioAssetUserData!(execute, Object=@n4) -> (then, CastFailed, AsAudio Asset User Data:Audio Asset User Data Object Reference) {is=title K2Node_DynamicCast}
   > n6 = Branch(Condition=HasTag!(
     TagContainer:"Gameplay Tag Container Structure (by ref)"=Get MetadataTags!(self:Audio Asset User Data Object Reference=@n5) -> (MetadataTags:Gameplay Tag Container Structure)
     Tag:Gameplay Tag Structure=Get MuteForStreamSafe() -> (MuteForStreamSafe:Gameplay Tag Structure)
     bExactMatch:bool=false
    ) -> (ReturnValue:bool))
   n7 = Get SoundBase() -> (SoundBase:Sound Base Object Reference)
   n8 = GetPluginNameForAssetPath!(AssetPath:String=Get Object Path String(Object=@n7)) -> (bIsInMainModule:bool, ReturnValue:String)
   n9 = "To String (Text)"(InText=Format Text!(
     Format="LOCTEXT(\"Audio.Metadata.StreamSafeExceptions.{PluginName}.{AssetName}\")"
     PluginName:String=@n8.ReturnValue
     AssetName:String=GetDisplayName(Object=@n7)
    ) -> (Result))
   > n10 = RequestGameplayTag!(execute, TagName:String=@n9, bErrorIfNotFound:bool=false) -> (then, OutTag:Gameplay Tag Structure, ReturnValue:bool)
   > n11 = Branch(Condition=@n10.ReturnValue)
   > RaiseScriptError!(execute, ErrorMessage:String=Asset already contains StreamSafe tags) -> (then)
   > Return Node!(execute, Success:bool=false)
   n12 = GetPluginBaseDir!(execute=[@n25.then, @n11.else], PluginName:String=@n8.ReturnValue) -> (then, OutBaseDir:String, ReturnValue:bool)
   > n13 = Branch(Condition=@n12.ReturnValue)
   n14 = "To String (Text)"(InText=Format Text!(Format="LOCTEXT(\"{PluginName}Tags.ini\")", PluginName:Text="To Text (String)"(InString=@n8.ReturnValue)) -> (Result))
   > n15 = AddNewGameplayTagSource!(
    execute
    NewTagSource:String=@n14
    RootDirToUse:String="To String (Text)"(InText=Format Text!(Format="LOCTEXT(\"{PluginBaseDir}/Config/Tags/\")", PluginBaseDir:String=@n12) -> (Result))
   ) -> (then, ReturnValue:bool)
   > n16 = Branch(Condition=@n15)
   n17 = String To Name(InString=@n14)
   > AddNewGameplayTagToINI!(execute, NewTag:String=@n9, Comment:String, TagSourceName:Name=@n17, bIsRestrictedTag:bool=false, bAllowNonRestrictedChildren:bool=true) -> (then, ReturnValue:bool)
   > n18 = RequestGameplayTag(TagName=@n9, bErrorIfNotFound=false)
   > n19 = Branch(Condition=@n18.ReturnValue)
   n20 = Get SoundBase()
   n21 = Get UserDataClass()
   n22 = GetAssetUserDataOfClass(execute=@n27.then, self=@n20, InUserDataClass=@n21)
   > n23 = Cast To AudioAssetUserData(Object=@n22)
   n24 = Get MetadataTags(self=@n23)
   AddGameplayTag!(execute=@n19.then, TagContainer:"Gameplay Tag Container Structure (by ref)"=@n24, Tag:Gameplay Tag Structure=@n18) -> (then)
   > Return Node(Success=@n18.ReturnValue)
   RaiseScriptError(execute=@n19.else, ErrorMessage="Request Gameplay Tag failed!")
   > Return Node(Success=false)
   RaiseScriptError(execute=@n16.else, ErrorMessage="Failed to add new Gameplay Tag source!")
   RaiseScriptError(execute=@n13.else, ErrorMessage=Failed to get plugin base directory.)
   n25 = AddGameplayTag(execute=[@n23.then, @n6.else], TagContainer=@n24, Tag=Get MuteForStreamSafe())
   n26 = Branch(execute=@n5.CastFailed, Condition="Not Equal (Name)"(A=@n17, B=FortniteGameTags.ini) -> (ReturnValue=~))
   > n27 = AddAssetUserDataOfClass!(execute, self:Interface Asset User Data Interface=@n20, InUserDataClass:Asset User Data Class Reference=@n21) -> (then, ReturnValue:bool)
   RaiseScriptError(execute=@n26.else, ErrorMessage=Sound must be located in a Game Feature Plugin.)
   > Return Node(Success=false)
   n28 = Show Message Dialog(
    execute=@n2.else
    Title=MakeLiteralText(Value="LOCTEXT(\"No Source Control Connected\")")
    Message=MakeLiteralText(Value="LOCTEXT(\"There is no source control enabled, continuing will not add the .ini file to your changelist which could lead to buildhealth issues. Would you like to continue?\")")
    MessageType=YesNo
    DefaultValue=No
    MessageCategory=Warning
   )
   > n29 = Switch on EAppReturnType(Selection=@n28)
   > Return Node(Success=false)
   n30 = Get Object Path String(Object=Get SoundBase())
   Format Text!(
    Format="LOCTEXT(\"{PluginName}Tags.ini\")"
    PluginName:Text="To Text (String)"(InString=GetPluginNameForAssetPath(AssetPath=@n30).ReturnValue)
   ) -> (Result)
   Return Node(execute=@n23.CastFailed, Success=false)
  ]}
 RemoveMuteForStreamSafeTag {nodes [
   n1 = RemoveMuteForStreamSafeTag!() -> (then, AudioAsset:Sound Base Object Reference) {is=title K2Node_FunctionEntry}
   > n2 = Set Soundbase(Soundbase:Sound Base Object Reference=@n1) -> (Output_Get:Sound Base Object Reference)
   > n3 = GetAssetUserDataOfClass(self=@n2, InUserDataClass=Get UserDataClass())
   > n4 = Cast To AudioAssetUserData(Object=@n3)
   > n5 = ResetEditorProperty!(execute, Object:Object Reference=@n4, PropertyName:Name=AssetUserData, ChangeNotifyMode:EPropertyAccessChangeNotifyMode Enum) -> (then, ReturnValue:bool)
   > Return Node(Success=@n5)
   Return Node(execute=@n4.CastFailed, Success=false)
  ]}
]
variables []
