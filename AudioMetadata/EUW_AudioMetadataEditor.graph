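# Editor utility widget for viewing a sound asset's user data and offering a MuteForStreamSafe tag.
# On Construct, a single property view binds to the widget's AudioAsset variable, and changes pass the sound to BP_AudioMetadataViewmodel via Set Sound.
# Switchers hold a prompt or details view and Add or Remove buttons; the graph never sets them or handles clicks.
# Remove is disabled, with a tooltip describing manual cleanup.
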EditorUtilityWidgetBlueprint /AudioMetadata/EUW_AudioMetadataEditor
graphs [EventGraph {nodes [
   "On Property Changed (AssetSelector)"!() -> (OutputDelegate, then, PropertyName:Name=None) {is=title K2Node_ComponentBoundEvent}
   > Set Sound!(
    execute
    self:Audio Asset User Data Viewmodel Object Reference=Get BP_AudioMetadataViewmodel() -> (BP_AudioMetadataViewmodel:Audio Asset User Data Viewmodel Object Reference)
    Sound:Sound Base Object Reference=Get AudioAsset() -> (AudioAsset:Sound Base Object Reference)
   ) -> (then) {is=title K2Node_CallFunction}
   Event Construct()
   > SetObject(self=Get AssetSelector() -> (AssetSelector:Single Property View Object Reference), NewObject=Self-Reference())
  ]}]
variables [AudioAsset {type=SoundBase, default=null}]
widgets {
 info {rootWidgetClass=&StackBox, widgetCount=12}
 widgets [
  StackBox_1 {widgetClassPath=&StackBox, widget=StackBox_1}
  StackBox_0 {
   widgetClassPath=&StackBox
   parent=StackBox_1
   widget=StackBox_0
   properties {orientation=Orient_Vertical}
   slot {class=StackBoxSlot, properties {padding=<ltrb 5 5 5 5>, size {sizeRule=Fill, value=0.25}}}
  }
  VerticalBox_4 {
   widgetClassPath=&VerticalBox
   parent=StackBox_0
   widget=VerticalBox_4
   slot {class=StackBoxSlot, properties {padding=<ltrb 5 5 5 5>, size {sizeRule=Fill, value=1}}}
  }
  AssetSelector {
   widgetClassPath=&SinglePropertyView
   parent=VerticalBox_4
   widget=AssetSelector
   properties {nameOverride=Asset, propertyName=AudioAsset}
   slot {class=VerticalBoxSlot, properties {verticalAlignment=VAlign_Top}}
  }
  UserDataViewSwitcher {
   widgetClassPath=&WidgetSwitcher
   bIsVariable=true
   parent=VerticalBox_4
   widget=UserDataViewSwitcher
   slot {class=VerticalBoxSlot, properties {size {sizeRule=Fill, value=0.5}}}
  }
  TextBlock_225 {
   widgetClassPath=&TextBlock
   parent=UserDataViewSwitcher
   widget=TextBlock_225
   properties {
    text=Select an asset.
    font {
     bForceMonospaced=false
     bMaterialIsStencil=false
     fontMaterial=null
     fontObject=&Roboto
     letterSpacing=0
     monospacedWidth=1
     size=12
     skewAmount=0
     typefaceFontName=Bold
     outlineSettings {
      bApplyOutlineToDropShadows=false
      bMiteredCorners=false
      bSeparateFillAlpha=false
      outlineMaterial=null
      outlineSize=0
      outlineColor=<rgba 0 0 0 1>
     }
    }
   }
   slot {class=WidgetSwitcherSlot, properties {horizontalAlignment=HAlign_Center, verticalAlignment=VAlign_Center, padding=<ltrb 0 5 0 5>}}
  }
  UserDataDetailsView {
   widgetClassPath=&DetailsView
   parent=UserDataViewSwitcher
   widget=UserDataDetailsView
   properties {bForceHiddenPropertyVisibility=true}
   slot {class=WidgetSwitcherSlot}
  }
  AddRemoveSwitcher {
   widgetClassPath=&WidgetSwitcher
   bIsVariable=true
   parent=VerticalBox_4
   widget=AddRemoveSwitcher
   slot {class=VerticalBoxSlot, properties {verticalAlignment=VAlign_Top}}
  }
  AddUserData {widgetClassPath=&EditorUtilityButton, parent=AddRemoveSwitcher, widget=AddUserData, slot {class=WidgetSwitcherSlot}}
  TextBlock {
   widgetClassPath=&TextBlock
   parent=AddUserData
   widget=TextBlock
   properties {
    text=Add MuteForStreamSafe Tag
    font {
     bForceMonospaced=false
     bMaterialIsStencil=false
     fontMaterial=null
     fontObject=&Roboto
     letterSpacing=0
     monospacedWidth=1
     size=12
     skewAmount=0
     typefaceFontName=Bold
     outlineSettings {
      bApplyOutlineToDropShadows=false
      bMiteredCorners=false
      bSeparateFillAlpha=false
      outlineMaterial=null
      outlineSize=0
      outlineColor=<rgba 0 0 0 1>
     }
    }
   }
   slot {class=ButtonSlot, properties {horizontalAlignment=HAlign_Left, verticalAlignment=VAlign_Fill, padding=<ltrb 5 5 5 5>}}
  }
  RemoveUserData {
   widgetClassPath=&EditorUtilityButton
   parent=AddRemoveSwitcher
   widget=RemoveUserData
   properties {
    bIsEnabled=false
    toolTipText="This tool currently doesn't support removal of StreamSafe tags. Please open the sound asset and clear the Asset User Data array, then save the asset. You may also want to revert the tags .ini file."
   }
   slot {class=WidgetSwitcherSlot}
  }
  TextBlock_1 {
   widgetClassPath=&TextBlock
   parent=RemoveUserData
   widget=TextBlock_1
   properties {
    text=Remove User Data
    font {
     bForceMonospaced=false
     bMaterialIsStencil=false
     fontMaterial=null
     fontObject=&Roboto
     letterSpacing=0
     monospacedWidth=1
     size=12
     skewAmount=0
     typefaceFontName=Bold
     outlineSettings {
      bApplyOutlineToDropShadows=false
      bMiteredCorners=false
      bSeparateFillAlpha=false
      outlineMaterial=null
      outlineSize=0
      outlineColor=<rgba 0 0 0 1>
     }
    }
   }
   slot {class=ButtonSlot, properties {horizontalAlignment=HAlign_Left, verticalAlignment=VAlign_Fill, padding=<ltrb 5 5 5 5>}}
  }
 ]
}
